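# Looping sound cue for the Zero Point Crystal attract sound, classed as an ambient loot loop.
# It reuses the crystal's Active loop wave, fades it in over about half a second with an Enveloper, then raises its pitch with a fixed-value Modulator.
# Overridden logarithmic attenuation with listener focus spatializes it, and a shared loot-ambient concurrency set governs simultaneous instances.
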SoundCue /Game/Sounds/Fort_GamePlay_Sounds/Player/ZeroPointCrystals/ZeroPointCrystal_Attract_Loop_Cue
root=Modulator_0
nodes [
 {
  id=Enveloper_0
  children [WavePlayer_0]
  properties {pitchCurve {editorCurveData {keys [{time=0}]}}, volumeCurve {editorCurveData {keys [<at 0.0172 0.0116>,<at 0.5444 1>]}}}
 }
 {id=Modulator_0, children [Enveloper_0], properties {pitchMax=1.4, pitchMin=1.4, volumeMax=1, volumeMin=1}}
 {
  id=WavePlayer_0
  properties {bLooping=true, soundWaveAssetPtr=&/ZeroPointCrystal_Active_Loop_01}
 }
]
properties {
 bOverrideAttenuation=true
 maxDistance=3400
 soundClassObject=/Game/Sounds/FortSoundClassesAndModes/Fort_Loot_Ambient_Loop
 volumeMultiplier=0.5
 attenuationOverrides {bEnableListenerFocus=true, distanceAlgorithm=Logarithmic, falloffDistance=3000, nonFocusVolumeAttenuation=0.4762}
 concurrencyOverrides {maxCount=2}
 concurrencySet [/Game/Sounds/Concurrency_Rules/Loot_Ambient_Conc]
}
