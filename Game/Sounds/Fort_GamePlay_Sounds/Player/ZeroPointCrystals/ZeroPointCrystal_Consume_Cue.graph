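# Sound cue for consuming a Zero Point Crystal, classed with gadget and item sounds.
# A single wave player feeds the root directly, with no randomization or layering.
# It plays at reduced volume through a dedicated Zero Point Crystal attenuation asset that governs its spatial falloff.
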SoundCue /Game/Sounds/Fort_GamePlay_Sounds/Player/ZeroPointCrystals/ZeroPointCrystal_Consume_Cue
root=WavePlayer_1
nodes [{id=WavePlayer_1, properties {soundWaveAssetPtr=&/ZeroPointCrystal_Consume_01}}]
properties {
 attenuationSettings=/Game/Sounds/Attenuation/Fort_ZeroPointCrystal_Att
 maxDistance=6500
 soundClassObject=&Fort_Gadgets_Items
 volumeMultiplier=0.55
}
